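# Anim Blueprint for the Vigilante Board back bling, posing the pack bone and adding secondary motion to its four wheels.
# EventGraph smooths jump and ground-movement alphas and converts owner spine_05-to-spine_01 yaw delta into pack roll.
# AnimGraph chains additive pack transforms (jump offset, constant offset, spine-driven roll) into low-alpha Anim Dynamics wheels.
# The speedAlpha-driven transform sits outside the output chain, so it has no effect.
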AnimBlueprint /BRCosmetics/Accessories/FORT_Backpacks/M_MED_VigilanteBoard_Backpack/Meshes/M_MED_Vigilante_Board_Pack_AnimBlueprint
target_skeleton=/Game/Accessories/FORT_Backpacks/M_MED_VigilanteBoard_Backpack/Meshes/M_MED_Vigilante_Board_Pack_Skeleton
graphs [
 AnimGraph {nodes [
   "Transform (Modify) Bone - Bone: pack"!(ComponentPose, Alpha=Get speedAlpha() -> (speedAlpha:double), Translation="0,-1.5,0", Rotation="0,0,0") -> (Pose) {
    rotationMode=BMM_Additive
    rotationSpace=BCS_BoneSpace
    translationMode=BMM_Additive
    translationSpace=BCS_BoneSpace
    alphaBoolBlend {blendOption=Linear}
    boneToModify {boneName=pack}
    translation {y=-1.5}
   }
   n1 = "Transform (Modify) Bone - Bone: pack"(
    ComponentPose=Local To Component(LocalPose=Local Space Ref Pose())
    Alpha=Get jumpAlpha() -> (jumpAlpha:double)
    Translation="0,-2.5,0"
    Rotation="0.000,0.000,-5"
   ) {
    rotationMode=BMM_Additive
    rotationSpace=BCS_BoneSpace
    translationMode=BMM_Additive
    translationSpace=BCS_BoneSpace
    alphaBoolBlend {blendOption=Linear}
    boneToModify {boneName=pack}
    rotation {roll=-5}
    translation {y=-2.5}
   }
   n2 = "Transform (Modify) Bone - Bone: pack"!(ComponentPose=@n1, Alpha) -> (Pose) {
    rotationMode=BMM_Additive
    rotationSpace=BCS_BoneSpace
    translationMode=BMM_Additive
    translationSpace=BCS_BoneSpace
    alphaBoolBlend {blendOption=Linear}
    boneToModify {boneName=pack}
    translation {y=-2.75}
   }
   n3 = "Transform (Modify) Bone - Bone: pack"!(ComponentPose=@n2, Alpha, Translation:Vector="0,0,0", Rotation:Rotator=Get backpack_rotate(), Scale:Vector) -> (Pose) {
    rotationMode=BMM_Additive
    rotationSpace=BCS_BoneSpace
    translationMode=BMM_Additive
    translationSpace=BCS_BoneSpace
    alphaBoolBlend {blendOption=Linear}
    boneToModify {boneName=pack}
   }
   n4 = Anim Dynamics - Bone: dyn_wheel_01(ComponentPose=@n3, Alpha=0.2, ExternalForce="0,0,0") {
    angularDampingOverride=0.96
    bGravityOverrideInSimSpace=true
    bOverrideAngularDamping=true
    bUseGravityOverride=true
    alphaBoolBlend {blendOption=Linear}
    boundBone {boneName=dyn_wheel_01}
    componentAppliedLinearAccClamp=<1000 1000 1000>
    componentLinearAccScale=<0.2 0.2 0.2>
    componentLinearVelScale=<0.1 0.1 0.1>
    gravityOverride {z=-980}
    physicsBodyDefinitions [{
      sphereCollisionRadius=0
      boundBone {boneName=dyn_wheel_01}
      boxExtents=<6 6 6>
      constraintSetup {twistAxis=AxisY, angularLimitsMax=<90 0 0>, angularLimitsMin=<-90 0 0>, angularTarget=<0 0 0>}
      localJointOffset=<0 -1 0.6>
     }]
   }
   n5 = Anim Dynamics - Bone: dyn_wheel_02(ComponentPose=@n4, Alpha=0.2, ExternalForce="0,0,0") {
    angularDampingOverride=0.96
    bGravityOverrideInSimSpace=true
    bOverrideAngularDamping=true
    bUseGravityOverride=true
    alphaBoolBlend {blendOption=Linear}
    boundBone {boneName=dyn_wheel_02}
    componentAppliedLinearAccClamp=<1000 1000 1000>
    componentLinearAccScale=<0.2 0.2 0.2>
    componentLinearVelScale=<0.1 0.1 0.1>
    gravityOverride {z=-980}
    physicsBodyDefinitions [{
      sphereCollisionRadius=0
      boundBone {boneName=dyn_wheel_02}
      boxExtents=<6 6 6>
      constraintSetup {twistAxis=AxisY, angularLimitsMax=<90 0 0>, angularLimitsMin=<-90 0 0>, angularTarget=<0 0 0>}
      localJointOffset=<0 -1 0.9>
     }]
   }
   n6 = Anim Dynamics - Bone: dyn_wheel_03(ComponentPose=@n5, Alpha=0.2, ExternalForce="0,0,0") {
    angularDampingOverride=0.96
    bGravityOverrideInSimSpace=true
    bOverrideAngularDamping=true
    bUseGravityOverride=true
    alphaBoolBlend {blendOption=Linear}
    boundBone {boneName=dyn_wheel_03}
    componentAppliedLinearAccClamp=<1000 1000 1000>
    componentLinearAccScale=<0.2 0.2 0.2>
    componentLinearVelScale=<0.1 0.1 0.1>
    gravityOverride {z=-980}
    physicsBodyDefinitions [{
      sphereCollisionRadius=0
      boundBone {boneName=dyn_wheel_03}
      boxExtents=<6 6 6>
      constraintSetup {twistAxis=AxisY, angularLimitsMax=<90 0 0>, angularLimitsMin=<-90 0 0>, angularTarget=<0 0 0>}
      localJointOffset=<0 -1 1.2>
     }]
   }
   n7 = Anim Dynamics - Bone: dyn_wheel_04(ComponentPose=@n6, Alpha=0.2, ExternalForce="0,0,0") {
    angularDampingOverride=0.96
    bGravityOverrideInSimSpace=true
    bOverrideAngularDamping=true
    bUseGravityOverride=true
    alphaBoolBlend {blendOption=Linear}
    boundBone {boneName=dyn_wheel_04}
    componentAppliedLinearAccClamp=<1000 1000 1000>
    componentLinearAccScale=<0.2 0.2 0.2>
    componentLinearVelScale=<0.1 0.1 0.1>
    gravityOverride {z=-980}
    physicsBodyDefinitions [{
      sphereCollisionRadius=0
      boundBone {boneName=dyn_wheel_04}
      boxExtents=<6 6 6>
      constraintSetup {twistAxis=AxisY, angularLimitsMax=<90 0 0>, angularLimitsMin=<-90 0 0>, angularTarget=<0 0 0>}
      localJointOffset=<0 -1 1.5>
     }]
   }
   Output Pose(Result=Component To Local(ComponentPose=@n7)) {layerGroup=None}
  ]}
 EventGraph {nodes [
   n1 = Event BlueprintUpdateAnimation()
   > n2 = Sequence!(execute) -> (then_0, then_1, then_2:Exec)
   > n3 = Branch(Condition=Get bIsJumping())
   n4 = Get jumpAlpha()
   > Set jumpAlpha(jumpAlpha:double=FInterpTo(Current=@n4, Target=1, DeltaTime=@n1.DeltaTimeX, InterpSpeed=2.5)) -> (Output_Get:double)
   Set jumpAlpha(execute=@n3.else, jumpAlpha=FInterpTo(Current=@n4, Target=0, DeltaTime=@n1.DeltaTimeX, InterpSpeed=2.5))
   n5 = Get PawnSpeed2D()
   n6 = Get bIsJumping()
   n7 = Branch(execute=@n2.then_1, Condition=AND Boolean(A="float > float"(A=@n5, B=0) {is=title K2Node_CallFunction}, B=NOT Boolean(A=@n6)))
   n8 = Get speedAlpha()
   > Set speedAlpha(speedAlpha:double=FInterpTo(Current=@n8, Target=1, DeltaTime=@n1.DeltaTimeX, InterpSpeed=10)) -> (Output_Get:double)
   Set speedAlpha(execute=@n7.else, speedAlpha=FInterpTo(Current=@n8, Target=0, DeltaTime=@n1.DeltaTimeX, InterpSpeed=2.5))
   n9 = GetOwnerSkeletalMeshComponent()
   n10 = BreakRotator(InRot=BreakTransform(InTransform=GetDeltaTransformFromRefPose(self=@n9, BoneName=spine_05, BaseName=spine_01)).Rotation)
   n11 = MapRangeClamped(Value=@n10.Yaw, InRangeA=0, InRangeB=30, OutRangeB=5)
   n12 = MapRangeClamped(Value=@n10.Yaw, InRangeA=0, InRangeB=-30, OutRangeB=25)
   Set backpack_rotate(execute=@n2.then_2, backpack_rotate=MakeRotator(Roll=float + float(A=@n12, B=@n11) {is=title K2Node_CommutativeAssociativeBinaryOperator}))
   TryGetPawnOwner()
  ]}
]
variables [
 backpack_rotate {type=Rotator}
 jumpAlpha {type=number}
 speedAlpha {type=number}
]
